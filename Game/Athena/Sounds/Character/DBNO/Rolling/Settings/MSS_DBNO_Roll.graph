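# Mono one-shot MetaSound for DBNO character rolling foley, layering random surface sounds with a pitch-randomized clothing sweetener.
# TriggerSwitch alternates two surface players, crossfading and carrying over playback position so surface changes continue mid-roll.
# An occlusion change while the source sits above the listener's height threshold switches to band-boosted, high-shelf-filtered surface audio without cloth.
# Affiliation sets final gain; the Mobile page uses one surface player without high-shelf.
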MetaSoundSource /Game/Athena/Sounds/Character/DBNO/Rolling/Settings/MSS_DBNO_Roll
inputs [
 AdditionalOperands {type=Float}
 BandPass.BandWidth {type=Float}
 BandPass.FreqCutOff {type=Float}
 Character.Affiliation {type=Enum:FortAudioAffiliation}
 Character.Health {type=Float}
 Character.UniqueID {type=Int32}
 Clothing Sweeteners {type=WaveAsset:Array}
 CrossfadeTime {type=Time}
 Enemy.Volume {type=Float}
 Friendly.Volume {type=Float}
 HighShelf.FreqCutOff {type=Float}
 HighShelf.Gain {type=Float}
 Random.Pitch.Base {type=Float}
 Random.Pitch.Spread {type=Float}
 Random.Volume.Base {type=Float}
 Random.Volume.Cloth Sweetener Volume {type=Float}
 Random.Volume.Spread {type=Float}
 SurfaceSounds {type=WaveAsset:Array}
 TriggerSwitch {type=Trigger}
 UE.Attenuation.Distance {type=Float}
 UE.Listener.Location.X {type=Float}
 UE.Listener.Location.Y {type=Float}
 UE.Listener.Location.Z {type=Float}
 UE.Source.Location.X {type=Float}
 UE.Source.Location.Y {type=Float}
 UE.Source.Location.Z {type=Float}
 UE.Source.Occlusion.IsOccluded {type=Bool}
 UE.Source.OnPlay {type=Trigger}
 VerticalHeightCheck {type=Float}
]
interfaces [
 Fort.Character:1.0
 UE.Attenuation:1.0
 UE.Listener.Location:1.0
 UE.OutputFormat.Mono:1.0
 UE.Source.Location:1.0
 UE.Source.Occlusion:1.0
 UE.Source.OneShot:1.0
 UE.Source:1.1
]
outputs [
 "UE.OutputFormat.Mono.Audio:0" {type=Audio}
 UE.Source.OneShot.OnFinished {type=Trigger}
]
pages [
 Default {
  is_build_graph=true
  input_defaults {
   AdditionalOperands=3
   BandPass.BandWidth=1.92
   BandPass.FreqCutOff=240
   Character.Affiliation=0
   Character.Health=0
   Character.UniqueID=0
   CrossfadeTime=0.5
   Enemy.Volume=0
   Friendly.Volume=-4.5
   HighShelf.FreqCutOff=5520
   HighShelf.Gain=-5.5521
   Random.Pitch.Base=0
   Random.Pitch.Spread=0.5
   Random.Volume.Base=9
   Random.Volume.Cloth Sweetener Volume=-3.1359
   Random.Volume.Spread=0
   TriggerSwitch=null
   UE.Attenuation.Distance=0
   UE.Listener.Location.X=0
   UE.Listener.Location.Y=0
   UE.Listener.Location.Z=0
   UE.Source.Location.X=0
   UE.Source.Location.Y=0
   UE.Source.Location.Z=0
   UE.Source.Occlusion.IsOccluded=null
   UE.Source.OnPlay=null
   VerticalHeightCheck=200
   Clothing Sweeteners []
   SurfaceSounds []
  }
  nodes [
   n1 = AbsoluteValue.Abs:Float(Input=^Random.Volume.Spread)
   n2 = UE.RandomFloat(
    Next=^UE.Source.OnPlay
    Max=UE.Add:Float(PrimaryOperand=^Random.Volume.Base, AdditionalOperands=@n1.Value).Out
    Min=UE.Subtract:Float(PrimaryOperand=^Random.Volume.Base, AdditionalOperands=@n1.Value).Out
   )
   DecibelstoLinearGain_4 = UE.Decibels to Linear Gain:Float(Decibels=@n2.Value)
   n3 = InitVariable.Float()
   TriggerSequence2_0 = "UE.Trigger Sequence (2)"(In=^TriggerSwitch)
   n4 = VariableAccessor.Trigger(Variable=VariableMutator.Trigger(Variable=InitVariable.Trigger().Variable, Value=@TriggerSequence2_0.Out 0).Variable)
   n5 = VariableAccessor.Trigger(Variable=@n4.Variable)
   n6 = VariableAccessor.Trigger(Variable=VariableMutator.Trigger(Variable=InitVariable.Trigger().Variable, Value=@TriggerSequence2_0.Out 1).Variable)
   n7 = VariableAccessor.Trigger(Variable=@n6.Variable)
   n8 = Value.Value:Float(Reset=@n5.Value, Set=@n7.Value) {also {defaults {Init Value=0, Target Value=1}}}
   n9 = UE.InterpTo:Audio(Interp Time=^CrossfadeTime, Target=@n8.Output Value)
   n10 = InitVariable.Float()
   n11 = Value.Value:Float(Set=@n5.Value, Reset=@n7.Value) {also {defaults {Init Value=0, Target Value=1}}}
   n12 = UE.InterpTo:Audio(Interp Time=^CrossfadeTime, Target=@n11.Output Value)
   ArrayRandomGet_2 = Array.Random Get:WaveAsset:Array(In Array=^SurfaceSounds, Next=@n4.Value) {also {defaults {No Repeats=-1, Weights []}}}
   n13 = InitVariable.Trigger()
   TriggerDelay_0 = UE.Trigger Delay(Delay Time=^CrossfadeTime, In=@n11.On Reset, Reset=@n11.On Set)
   n14 = VariableDeferredAccessor.Time(Variable=InitVariable.Time().Variable)
   WavePlayer_0 = UE.Wave Player:Mono(
    Play=@ArrayRandomGet_2.On Next
    Wave Asset=@ArrayRandomGet_2.Value
    Stop=VariableAccessor.Trigger(Variable=VariableMutator.Trigger(Variable=@n13.Variable, Value=@TriggerDelay_0.Out).Variable).Value
    Start Time=@n14.Value
   ) {also {version [1,1]}}
   ArrayRandomGet_0 = Array.Random Get:WaveAsset:Array(In Array=^SurfaceSounds, Next=@n6.Value) {also {defaults {No Repeats=-1, Weights []}}}
   n15 = InitVariable.Trigger()
   TriggerDelay_2 = UE.Trigger Delay(Delay Time=^CrossfadeTime, In=@n8.On Reset, Reset=@n8.On Set)
   n16 = VariableDeferredAccessor.Time(Variable=@n14.Variable)
   WavePlayer_1 = UE.Wave Player:Mono(
    Play=@ArrayRandomGet_0.On Next
    Wave Asset=@ArrayRandomGet_0.Value
    Stop=VariableAccessor.Trigger(Variable=VariableMutator.Trigger(Variable=@n15.Variable, Value=@TriggerDelay_2.Out).Variable).Value
    Start Time=@n16.Value
   ) {also {version [1,1]}}
   AudioMixerAudioMixerMono2_0 = "AudioMixer.Audio Mixer (Mono, 2)"(
    Gain 1=VariableAccessor.Float(Variable=VariableMutator.Float(Variable=@n3.Variable, Value=@n9.Value).Variable).Value
    Gain 0=VariableAccessor.Float(Variable=VariableMutator.Float(Variable=@n10.Variable, Value=@n12.Value).Variable).Value
    In 0=@WavePlayer_0.Out Mono
    In 1=@WavePlayer_1.Out Mono
   )
   n17 = VariableAccessor.Audio(Variable=VariableMutator.Audio(Value=@AudioMixerAudioMixerMono2_0.Out, Variable=InitVariable.Audio().Variable).Variable)
   n18 = VariableAccessor.Audio(Variable=VariableMutator.Audio(
     Variable=InitVariable.Audio().Variable
     Value=UE.Multiply:Audio by Float(AdditionalOperands=@DecibelstoLinearGain_4.Linear Gain, PrimaryOperand=@n17.Value).Out
    ).Variable)
   BiquadFilter_1 = UE.Biquad Filter:Audio(Bandwidth=^BandPass.BandWidth, Cutoff Frequency=^BandPass.FreqCutOff, In=@n18.Value) {also {defaults {Type=2}}}
   n19 = VariableAccessor.Audio(Variable=@n18.Variable)
   BiquadFilter_0 = UE.Biquad Filter:Audio(
    Cutoff Frequency=^HighShelf.FreqCutOff
    Gain=^HighShelf.Gain
    In=UE.Add:Audio(
     AdditionalOperands=UE.Multiply:Audio by Float(AdditionalOperands=^AdditionalOperands, PrimaryOperand=@BiquadFilter_1.Out).Out
     PrimaryOperand=@n19.Value
    ).Out
   ) {also {defaults {Type=6}}}
   n20 = UE.Reroute(Value=UE.Reroute(Value=^UE.Source.Occlusion.IsOccluded).Value)
   TriggerOnValueChange_0 = UE.Trigger On Value Change:Bool(Value=^UE.Source.Occlusion.IsOccluded)
   n21 = TriggerCompare.Clamp:Float(
    A=^UE.Source.Location.Z
    B=UE.Add:Float(PrimaryOperand=^UE.Listener.Location.Z, AdditionalOperands=^VerticalHeightCheck).Out
    Compare=@TriggerOnValueChange_0.Trigger
   ) {also {defaults {Type=3}}}
   n22 = VariableAccessor.Trigger(Variable=VariableMutator.Trigger(
     Variable=InitVariable.Trigger().Variable
     Value=TriggerCompare.Clamp:Bool(A=@n20.Value, Compare=@n21.True) {also {defaults {B="<MetaSoundBuilderResult.SUCCEEDED: 0>"}}}.True
    ).Variable)
   TriggerRouteTriggerRouteAudio2_0 = "TriggerRoute.Trigger Route (Audio, 2):Audio"(Set 0=^UE.Source.OnPlay, Value 1=@BiquadFilter_0.Out, Value 0=VariableAccessor.Audio(Variable=@n19.Variable).Value, Set 1=@n22.Value)
   n23 = InitVariable.Audio()
   DecibelstoLinearGain_2 = UE.Decibels to Linear Gain:Float(Decibels=^Random.Volume.Cloth Sweetener Volume)
   DecibelstoLinearGain_1 = UE.Decibels to Linear Gain:Float(Decibels=@n2.Value)
   ArrayRandomGet_1 = Array.Random Get:WaveAsset:Array(In Array=^Clothing Sweeteners, Next=^UE.Source.OnPlay) {also {defaults {No Repeats=-1, Weights []}}}
   n24 = InitVariable.Float()
   n25 = AbsoluteValue.Abs:Float(Input=^Random.Pitch.Spread)
   n26 = UE.RandomFloat(
    Next=^UE.Source.OnPlay
    Max=UE.Add:Float(PrimaryOperand=^Random.Pitch.Base, AdditionalOperands=@n25.Value).Out
    Min=UE.Subtract:Float(PrimaryOperand=^Random.Pitch.Base, AdditionalOperands=@n25.Value).Out
   )
   WavePlayer_2 = UE.Wave Player:Mono(
    Play=@ArrayRandomGet_1.On Next
    Wave Asset=@ArrayRandomGet_1.Value
    Pitch Shift=VariableAccessor.Float(Variable=VariableMutator.Float(Variable=@n24.Variable, Value=@n26.Value).Variable).Value
   ) {also {version [1,1]}}
   n27 = VariableMutator.Audio(Variable=InitVariable.Audio().Variable, Value=@WavePlayer_2.Out Mono)
   n28 = UE.Multiply:Audio by Float(
    AdditionalOperands=@DecibelstoLinearGain_2.Linear Gain
    PrimaryOperand=UE.Multiply:Audio by Float(AdditionalOperands=@DecibelstoLinearGain_1.Linear Gain, PrimaryOperand=VariableAccessor.Audio(Variable=@n27.Variable).Value).Out
   )
   TriggerRouteTriggerRouteAudio2_1 = "TriggerRoute.Trigger Route (Audio, 2):Audio"(
    Set 0=^UE.Source.OnPlay
    Value 0=VariableAccessor.Audio(Variable=VariableMutator.Audio(Variable=@n23.Variable, Value=@n28.Out).Variable).Value
    Set 1=VariableAccessor.Trigger(Variable=@n22.Variable).Value
   )
   n29 = InitVariable.Float()
   DecibelstoLinearGain_0 = UE.Decibels to Linear Gain:Float(Decibels=^Enemy.Volume)
   DecibelstoLinearGain_3 = UE.Decibels to Linear Gain:Float(Decibels=^Friendly.Volume)
   n30 = &TriggerCompareAffiliation(
    A=^Character.Affiliation
    Compare=^UE.Source.OnPlay
    B=VariableAccessor.Enum:FortAudioAffiliation(Variable=InitVariable.Enum:FortAudioAffiliation().Variable).Value
   )
   TriggerRouteTriggerRouteFloat2_0 = "TriggerRoute.Trigger Route (Float, 2):Float"(Value 1=@DecibelstoLinearGain_0.Linear Gain, Value 0=@DecibelstoLinearGain_3.Linear Gain, Set 1=@n30.False, Set 0=@n30.True)
   n31 = Value.Value:Float(Set=@TriggerRouteTriggerRouteFloat2_0.On Set, Target Value=@TriggerRouteTriggerRouteFloat2_0.Value) {also {defaults {Init Value=1}}}
   Multiply_4 = UE.Multiply:Audio by Float(
    PrimaryOperand=UE.Add:Audio(AdditionalOperands=@TriggerRouteTriggerRouteAudio2_0.Value, PrimaryOperand=@TriggerRouteTriggerRouteAudio2_1.Value).Out
    AdditionalOperands=VariableAccessor.Float(Variable=VariableMutator.Float(Variable=@n29.Variable, Value=@n31.Output Value).Variable).Value
   )
   n32 = InitVariable.Trigger()
   GetWaveDuration_0 = "UE.Get Wave Duration: "(Wave=@ArrayRandomGet_2.Value)
   GetWaveDuration_1 = "UE.Get Wave Duration: "(Wave=@ArrayRandomGet_0.Value)
   n33 = InitVariable.Trigger()
   n34 = InitVariable.Trigger()
   TriggerRouteTriggerRouteTime2_0 = "TriggerRoute.Trigger Route (Time, 2):Time"(
    Value 0=@GetWaveDuration_0.Duration
    Value 1=@GetWaveDuration_1.Duration
    Set 0=VariableAccessor.Trigger(Variable=VariableMutator.Trigger(Value=@ArrayRandomGet_2.On Next, Variable=@n33.Variable).Variable).Value
    Set 1=VariableAccessor.Trigger(Variable=VariableMutator.Trigger(Value=@ArrayRandomGet_0.On Next, Variable=@n34.Variable).Variable).Value
   )
   TriggerOnce_0 = UE.Trigger Once(Trigger In=@TriggerRouteTriggerRouteTime2_0.On Set)
   TriggerDelay_1 = UE.Trigger Delay(
    Delay Time=UE.Add:Time(PrimaryOperand=@TriggerRouteTriggerRouteTime2_0.Value) {also {defaults {AdditionalOperands=0.2}}}.Out
    In=@TriggerOnce_0.Trigger Out
   )
   n35 = InitVariable.Trigger()
   TriggerAccumulatorTriggerAccumulate2_0 = "TriggerAccumulator.Trigger Accumulate (2)"(
    In 1=VariableAccessor.Trigger(Variable=VariableMutator.Trigger(Variable=@n32.Variable, Value=@TriggerDelay_1.Out).Variable).Value
    In 0=VariableAccessor.Trigger(Variable=VariableMutator.Trigger(Variable=@n35.Variable, Value=@WavePlayer_2.On Finished).Variable).Value
   )
   TriggerRouteTriggerRouteTime2_1 = "TriggerRoute.Trigger Route (Time, 2):Time"(
    Set 0=VariableAccessor.Trigger(Variable=@n7.Variable).Value
    Set 1=VariableAccessor.Trigger(Variable=@n5.Variable).Value
    Value 0=@WavePlayer_0.Playback Time
    Value 1=@WavePlayer_1.Playback Time
   )
   n36 = UE.ConversionTimeToFloat(In=@TriggerRouteTriggerRouteTime2_1.Value)
   VariableMutator.Time(
    Value=UE.ConversionFloatToTime(In=Value.Value:Float(Target Value=@n36.Out, Set=@TriggerRouteTriggerRouteTime2_1.On Set).Output Value).Out
    Variable=@n16.Variable
   )
  ]
  links [
   {dst="@UE.OutputFormat.Mono.Audio:0", src=Multiply_4.Out}
   {dst="@UE.Source.OneShot.OnFinished", src=TriggerAccumulatorTriggerAccumulate2_0.Out}
  ]
 }
 Mobile {
  nodes [
   n1 = VariableAccessor.Trigger(Variable=VariableMutator.Trigger(Value=^TriggerSwitch, Variable=InitVariable.Trigger().Variable).Variable)
   ArrayRandomGet_1 = Array.Random Get:WaveAsset:Array(In Array=^SurfaceSounds, Next=@n1.Value) {also {defaults {No Repeats=-1, Weights []}}}
   GetWaveDuration_0 = "UE.Get Wave Duration: "(Wave=@ArrayRandomGet_1.Value)
   InitVariable.Float()
   InitVariable.Trigger()
   InitVariable.Trigger()
   InitVariable.Trigger()
   n2 = AbsoluteValue.Abs:Float(Input=^Random.Volume.Spread)
   n3 = UE.RandomFloat(
    Next=^UE.Source.OnPlay
    Max=UE.Add:Float(PrimaryOperand=^Random.Volume.Base, AdditionalOperands=@n2.Value).Out
    Min=UE.Subtract:Float(PrimaryOperand=^Random.Volume.Base, AdditionalOperands=@n2.Value).Out
   )
   DecibelstoLinearGain_4 = UE.Decibels to Linear Gain:Float(Decibels=@n3.Value)
   n4 = InitVariable.Float()
   n5 = Value.Value:Float(Set=VariableAccessor.Trigger(Variable=@n1.Variable).Value) {also {defaults {Init Value=0, Target Value=1}}}
   n6 = UE.InterpTo:Audio(Interp Time=^CrossfadeTime, Target=@n5.Output Value)
   n7 = InitVariable.Trigger()
   TriggerDelay_0 = UE.Trigger Delay(Delay Time=^CrossfadeTime, In=@n5.On Reset, Reset=@n5.On Set)
   WavePlayer_0 = UE.Wave Player:Mono(
    Play=@ArrayRandomGet_1.On Next
    Wave Asset=@ArrayRandomGet_1.Value
    Stop=VariableAccessor.Trigger(Variable=VariableMutator.Trigger(Variable=@n7.Variable, Value=@TriggerDelay_0.Out).Variable).Value
    Start Time=VariableDeferredAccessor.Time(Variable=InitVariable.Time().Variable).Value
   ) {also {version [1,1]}}
   AudioMixerAudioMixerMono2_0 = "AudioMixer.Audio Mixer (Mono, 2)"(
    Gain 0=VariableAccessor.Float(Variable=VariableMutator.Float(Variable=@n4.Variable, Value=@n6.Value).Variable).Value
    In 0=@WavePlayer_0.Out Mono
   )
   n8 = VariableAccessor.Audio(Variable=VariableMutator.Audio(Value=@AudioMixerAudioMixerMono2_0.Out, Variable=InitVariable.Audio().Variable).Variable)
   n9 = VariableAccessor.Audio(Variable=VariableMutator.Audio(
     Variable=InitVariable.Audio().Variable
     Value=UE.Multiply:Audio by Float(AdditionalOperands=@DecibelstoLinearGain_4.Linear Gain, PrimaryOperand=@n8.Value).Out
    ).Variable)
   BiquadFilter_0 = UE.Biquad Filter:Audio(Bandwidth=^BandPass.BandWidth, Cutoff Frequency=^BandPass.FreqCutOff, In=@n9.Value) {also {defaults {Type=2}}}
   n10 = VariableAccessor.Audio(Variable=@n9.Variable)
   n11 = UE.Reroute(Value=UE.Reroute(Value=^UE.Source.Occlusion.IsOccluded).Value)
   TriggerOnValueChange_0 = UE.Trigger On Value Change:Bool(Value=^UE.Source.Occlusion.IsOccluded)
   n12 = TriggerCompare.Clamp:Float(
    A=^UE.Source.Location.Z
    B=UE.Add:Float(PrimaryOperand=^UE.Listener.Location.Z, AdditionalOperands=^VerticalHeightCheck).Out
    Compare=@TriggerOnValueChange_0.Trigger
   ) {also {defaults {Type=3}}}
   n13 = VariableAccessor.Trigger(Variable=VariableMutator.Trigger(
     Variable=InitVariable.Trigger().Variable
     Value=TriggerCompare.Clamp:Bool(A=@n11.Value, Compare=@n12.True) {also {defaults {B="<MetaSoundBuilderResult.SUCCEEDED: 0>"}}}.True
    ).Variable)
   TriggerRouteTriggerRouteAudio2_0 = "TriggerRoute.Trigger Route (Audio, 2):Audio"(
    Set 0=^UE.Source.OnPlay
    Value 1=UE.Add:Audio(
     AdditionalOperands=UE.Multiply:Audio by Float(AdditionalOperands=^AdditionalOperands, PrimaryOperand=@BiquadFilter_0.Out).Out
     PrimaryOperand=@n10.Value
    ).Out
    Value 0=VariableAccessor.Audio(Variable=@n10.Variable).Value
    Set 1=@n13.Value
   )
   n14 = InitVariable.Audio()
   DecibelstoLinearGain_2 = UE.Decibels to Linear Gain:Float(Decibels=^Random.Volume.Cloth Sweetener Volume)
   DecibelstoLinearGain_1 = UE.Decibels to Linear Gain:Float(Decibels=@n3.Value)
   ArrayRandomGet_0 = Array.Random Get:WaveAsset:Array(In Array=^Clothing Sweeteners, Next=^UE.Source.OnPlay) {also {defaults {No Repeats=-1, Weights []}}}
   n15 = InitVariable.Float()
   n16 = AbsoluteValue.Abs:Float(Input=^Random.Pitch.Spread)
   n17 = UE.RandomFloat(
    Next=^UE.Source.OnPlay
    Max=UE.Add:Float(PrimaryOperand=^Random.Pitch.Base, AdditionalOperands=@n16.Value).Out
    Min=UE.Subtract:Float(PrimaryOperand=^Random.Pitch.Base, AdditionalOperands=@n16.Value).Out
   )
   WavePlayer_1 = UE.Wave Player:Mono(
    Play=@ArrayRandomGet_0.On Next
    Wave Asset=@ArrayRandomGet_0.Value
    Pitch Shift=VariableAccessor.Float(Variable=VariableMutator.Float(Variable=@n15.Variable, Value=@n17.Value).Variable).Value
   ) {also {version [1,1]}}
   n18 = VariableMutator.Audio(Variable=InitVariable.Audio().Variable, Value=@WavePlayer_1.Out Mono)
   n19 = UE.Multiply:Audio by Float(
    AdditionalOperands=@DecibelstoLinearGain_2.Linear Gain
    PrimaryOperand=UE.Multiply:Audio by Float(AdditionalOperands=@DecibelstoLinearGain_1.Linear Gain, PrimaryOperand=VariableAccessor.Audio(Variable=@n18.Variable).Value).Out
   )
   TriggerRouteTriggerRouteAudio2_1 = "TriggerRoute.Trigger Route (Audio, 2):Audio"(
    Set 0=^UE.Source.OnPlay
    Value 0=VariableAccessor.Audio(Variable=VariableMutator.Audio(Variable=@n14.Variable, Value=@n19.Out).Variable).Value
    Set 1=VariableAccessor.Trigger(Variable=@n13.Variable).Value
   )
   n20 = InitVariable.Float()
   DecibelstoLinearGain_0 = UE.Decibels to Linear Gain:Float(Decibels=^Enemy.Volume)
   DecibelstoLinearGain_3 = UE.Decibels to Linear Gain:Float(Decibels=^Friendly.Volume)
   n21 = &TriggerCompareAffiliation(
    A=^Character.Affiliation
    Compare=^UE.Source.OnPlay
    B=VariableAccessor.Enum:FortAudioAffiliation(Variable=InitVariable.Enum:FortAudioAffiliation().Variable).Value
   )
   TriggerRouteTriggerRouteFloat2_0 = "TriggerRoute.Trigger Route (Float, 2):Float"(Value 1=@DecibelstoLinearGain_0.Linear Gain, Value 0=@DecibelstoLinearGain_3.Linear Gain, Set 1=@n21.False, Set 0=@n21.True)
   n22 = Value.Value:Float(Set=@TriggerRouteTriggerRouteFloat2_0.On Set, Target Value=@TriggerRouteTriggerRouteFloat2_0.Value) {also {defaults {Init Value=1}}}
   Multiply_4 = UE.Multiply:Audio by Float(
    PrimaryOperand=UE.Add:Audio(AdditionalOperands=@TriggerRouteTriggerRouteAudio2_0.Value, PrimaryOperand=@TriggerRouteTriggerRouteAudio2_1.Value).Out
    AdditionalOperands=VariableAccessor.Float(Variable=VariableMutator.Float(Variable=@n20.Variable, Value=@n22.Output Value).Variable).Value
   )
   n23 = InitVariable.Trigger()
   n24 = InitVariable.Trigger()
   TriggerAccumulatorTriggerAccumulate2_0 = "TriggerAccumulator.Trigger Accumulate (2)"(
    In 1=VariableAccessor.Trigger(Variable=VariableMutator.Trigger(Variable=@n23.Variable, Value=@WavePlayer_0.On Finished).Variable).Value
    In 0=VariableAccessor.Trigger(Variable=VariableMutator.Trigger(Variable=@n24.Variable, Value=@WavePlayer_1.On Finished).Variable).Value
   )
   VariableMutator.Trigger(Value=@ArrayRandomGet_1.On Next, Variable=InitVariable.Trigger().Variable)
  ]
  links [
   {dst="@UE.OutputFormat.Mono.Audio:0", src=Multiply_4.Out}
   {dst="@UE.Source.OneShot.OnFinished", src=TriggerAccumulatorTriggerAccumulate2_0.Out}
  ]
 }
]
properties {
 assetImportData=&self:AssetImportData
 attenuationSettings=&/ATT_DBNO_Roll
 audioPropertiesSheet=&/APS_Foley_DBNORoll
 maxDistance=3584
 sampleRate=48000
 soundClassObject=&Fort_Player_Foley_Athena
 volume=1.5
}
